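# Sound cue for the tire bounce of the tire fall trap, in the Fort Traps sound class.
# A single wave player feeds a modulator that slightly randomizes pitch on each play.
# It uses the shared trap attenuation settings and overrides concurrency to cap simultaneous instances at a low count, blocking new plays beyond that.
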SoundCue /Game/Sounds/Fort_Traps/Tire_Fall_Trap/Tire_Trap_Tires_Bounce_Cue
root=Modulator_0
nodes [
 {id=Modulator_0, children [WavePlayer_0], properties {pitchMax=1.15, pitchMin=0.9}}
 {id=WavePlayer_0, properties {soundWaveAssetPtr=&/Tire_Trap_Tires_Bounce}}
]
properties {
 attenuationSettings=&Fort_Traps_Att
 bOverrideConcurrency=true
 maxDistance=6000
 soundClassObject=&Fort_Traps
 concurrencyOverrides {maxCount=2, resolutionRule=PreventNew}
}
